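; GA_GrantUrgentQuest is a gameplay ability that grants an urgent quest to the owning player, under the Season 20 Charge Backpack narrative quest.
; On activation it casts the owning pawn's controller to FortPlayerControllerAthena and grants the configured quest definition through that controller's transient quests component.
; It then prints a debug confirmation string.
; The failed passive activation event only calls the parent, and the end-ability event is empty.
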(event Ability|EventFailedToActivatePassiveAbility
  (params (ActorInfo :type Gameplay%20Ability%20Actor%20Info%20Structure))
  (|CallParentFunction))

(event Ability|EventActivateAbility
  (|CallParentFunction)
  (bind _asfort_player_controller_athena (Utilities|Casting|CastToFortPlayerControllerAthena (Fortnite|GetFortPlayerControllerfromActor (Variables|GAVars|Data|GetOwningPlayerPawn))))
  (Class|FortControllerComponentTransientQuests|GrantTransientQuest (Class|FortPlayerControllerAthena|GetTransientQuestsComponent _asfort_player_controller_athena) (Variables|Default|GetQuest%20Def%20to%20Grant) 0)
  (Development|PrintString "Urgentquest granted" true true "(R=1.000000,G=0.000000,B=0.192244,A=1.000000)"))

(event Ability|EventOnEndAbility
  (params (bWasCancelled :type Boolean)))
