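; Gameplay ability supporting the urgent quest HUD for a Season 20 Charge Backpack narrative quest (W11 Q2).
; On activation it caches the avatar pawn and, on the server, binds to the Athena quest manager's urgent transient quest completed and failed events.
; Handlers match the ability's quest definition by primary asset ID; completion also calls CleanupCompletedQuest on the pawn's transient quests component.
; Both then run Cleanup to unbind the events.
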(event Ability|EventFailedToActivatePassiveAbility
  (params (ActorInfo :type Gameplay%20Ability%20Actor%20Info%20Structure))
  ;[1 1 1] Urgent Quest HUD setup
  ;;[0 0 0] Activation Failed
  (|CallParentFunction))

(event Ability|EventActivateAbility
  (bind _owning_player (Variables|Default|GetOwning%20Player))
  (bind _returnvalue (Fortnite|GetFortPlayerControllerfromActor _owning_player))
  (bind _returnvalue_2 (Quests|GetQuestManager _returnvalue "Athena"))
  (bind _returnvalue_3 (Class|FortQuestManager|GetQuestManagerComponent _returnvalue_2 "/Script/FortniteGame.FortQuestManagerComponent_Athena"))
  ;[1 1 1] Urgent Quest HUD setup
  ;;[0 0 0] Activate
  (bind _asfort_player_pawn_athena (Utilities|Casting|CastToFortPlayerPawnAthena (Ability|GetAvatarActorfromActorInfo))
    (:then
      (Variables|Default|SetOwning%20Player _asfort_player_pawn_athena)
      (if (Networking|IsServer)
        (Quests|BindEventtoOnUrgentTransientQuestCompleted _returnvalue_3 (EventDispatchers|CreateEvent :!event Custom|Quest%20Completed))
        (Quests|BindEventto%28DEPRECATED%29OnUrgentTransientQuestFailed _returnvalue_3 (EventDispatchers|CreateEvent :!event Custom|OnQuestFailed))))
    (:CastFailed
      (Ability|EndAbility))))

(event Custom|Quest%20Completed
  (params (TransientQuestItem :type Fort%20Quest%20Item%20Object%20Reference))
  ;[1 1 1] Urgent Quest HUD setup
  ;;[0 0 0] Quest Completed
  (if (AssetManager|Equal(PrimaryAssetId) (AssetManager|GetPrimaryAssetIdfromObject (Quest|GetQuestDefinition TransientQuestItem)) (AssetManager|GetPrimaryAssetIdfromObject (Variables|FortTransientQuestGameplayAbility|GetQuestItemDefinition)))
    (Utilities|Casting|CastToFortUrgentQuestItemDefinition (Variables|FortTransientQuestGameplayAbility|GetQuestItemDefinition))
    (Class|FortControllerComponentTransientQuests|CleanupCompletedQuest (Class|FortPlayerPawn|GetTransientQuestsComponent (Variables|Default|GetOwning%20Player)) (Variables|FortTransientQuestGameplayAbility|GetQuestItemDefinition))
    (CallFunction|Cleanup)))

(event Custom|OnQuestFailed
  (params (TransientQuestItem :type Fort%20Quest%20Item%20Object%20Reference))
  ;[1 1 1] Urgent Quest HUD setup
  ;;[0 0 0] Clean up the quest HUD if it fails
  (if (AssetManager|Equal(PrimaryAssetId) (AssetManager|GetPrimaryAssetIdfromObject (Quest|GetQuestDefinition TransientQuestItem)) (AssetManager|GetPrimaryAssetIdfromObject (Variables|FortTransientQuestGameplayAbility|GetQuestItemDefinition)))
    (CallFunction|Cleanup)))

(event Custom|Cleanup
  (bind _owning_player (Variables|Default|GetOwning%20Player))
  (bind _returnvalue (Fortnite|GetFortPlayerControllerfromActor _owning_player))
  (bind _returnvalue_2 (Quests|GetQuestManager _returnvalue "Invalid"))
  (bind _returnvalue_3 (Class|FortQuestManager|GetQuestManagerComponent _returnvalue_2 "/Script/FortniteGame.FortQuestManagerComponent_Athena"))
  ;[1 1 1] Urgent Quest HUD setup
  ;;[0 0 0] Cleanup Binds
  (Quests|UnbindEventfromOnUrgentTransientQuestCompleted _returnvalue_3 (EventDispatchers|CreateEvent :!event Custom|Quest%20Completed))
  (Quests|UnbindallEventsfrom%28DEPRECATED%29OnUrgentTransientQuestFailed _returnvalue_3))
